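# Sound cue for the Assault Specialist NPC's holo bubble, playing a single hologram beam wave as a gadget-class sound.
# A modulator adds slight random pitch and volume variation to the wave player.
# It borrows the decoy grenade explosion attenuation asset and carries its own concurrency override that caps simultaneous voices and scales their volume by distance.
# The cue's debug flag is left on.
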SoundCue /NPCSpecialist_Assault/Audio/SC_NPC_Specialist_Assault_HoloBubble_Cue
root=Modulator_0
nodes [
 {id=Modulator_0, children [WavePlayer_1], properties {pitchMax=1.005, pitchMin=1, volumeMax=1, volumeMin=0.88}}
 {id=WavePlayer_1, properties {soundWaveAssetPtr=&/SC_AISpecialist_Assault_HologramBeam_01}}
]
properties {
 attenuationSettings=/Game/Sounds/Fort_Weapons/Grenades/Decoy/DecoyGrenade_Explo_Att
 bDebug=true
 bOverrideConcurrency=true
 maxDistance=13000
 soundClassObject=&Fort_Gadgets
 volumeMultiplier=1.8
 attenuationOverrides {falloffDistance=512, nonSpatializedRadiusStart=20, stereoSpread=420}
 concurrencyOverrides {maxCount=2, voiceStealReleaseTime=0.3, volumeScale=0.4, volumeScaleAttackTime=0.1, volumeScaleMode=Distance}
}
